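# Translucent red Default Lit material stored with the bounds meshes of the medium male player body.
# It holds an unconnected-as-shown pair of parameters: a red vector and an "opacity" scalar defaulting to half.
# Usage flags enable it on skeletal meshes and with mesh deformers.
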Material /Game/Characters/Player/Male/Medium/Bodies/M_MED_Bounds/Meshes/Bounds_Transparent_Red_Mat
graph {
 properties {
  bUsedWithMeshDeformer=true
  bUsedWithSkeletalMesh=true
  blendMode=BLEND_Translucent
  pixelDepthOffsetMode=PDOM_Legacy
  usedShadingModels=Default Lit
 }
 nodes [
  ScalarParameter() {defaultValue=0.5, parameterName=opacity}
  VectorParameter() {defaultValue=<rgba 1 0 0 0>}
 ]
}
